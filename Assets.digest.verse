# Copyright Epic Games, Inc. All Rights Reserved.
#################################################
# Generated Digest of Verse API
# DO NOT modify this manually!
# Generated from build: ++Fortnite+Release-36.20-CL-43847582
#################################################


using {/Verse.org/Simulation}
using {/UnrealEngine.com/Temporary/UI}
using {/Verse.org/Assets}
using {/Verse.org/Native}
using {/Verse.org/SceneGraph}
# Module import path: /playkigai.dig@fortnite.com/Minigame_PlungerDunger/GameManager
GameManager := module:
    @import_as("/Minigame_PlungerDunger/GameManager/game_manager.game_manager_C")
    game_manager<scoped {/playkigai.dig@fortnite.com/Minigame_PlungerDunger}> := class<final>(entity):


    game_manager_asset<scoped {/playkigai.dig@fortnite.com/Minigame_PlungerDunger}>:entity_prefab = external {}

# Module import path: /playkigai.dig@fortnite.com/Minigame_PlungerDunger/TargetSystem
TargetSystem := module:
    @import_as("/Minigame_PlungerDunger/TargetSystem/explosion.explosion_C")
    explosion<scoped {/playkigai.dig@fortnite.com/Minigame_PlungerDunger}> := class<final>(entity):


    explosion_asset<scoped {/playkigai.dig@fortnite.com/Minigame_PlungerDunger}>:entity_prefab = external {}

    NewNiagaraSystem_asset<scoped {/playkigai.dig@fortnite.com/Minigame_PlungerDunger}>:particle_system = external {}

    NewNiagaraSystem<scoped {/playkigai.dig@fortnite.com/Minigame_PlungerDunger}> := class<final><public>(particle_system_component):


    @import_as("/Minigame_PlungerDunger/TargetSystem/rail.rail_C")
    rail<scoped {/playkigai.dig@fortnite.com/Minigame_PlungerDunger}> := class<final>(entity):


    rail_asset<scoped {/playkigai.dig@fortnite.com/Minigame_PlungerDunger}>:entity_prefab = external {}

    @import_as("/Minigame_PlungerDunger/TargetSystem/target.target_C")
    target<scoped {/playkigai.dig@fortnite.com/Minigame_PlungerDunger}> := class<final>(entity):


    target_asset<scoped {/playkigai.dig@fortnite.com/Minigame_PlungerDunger}>:entity_prefab = external {}

# Module import path: /playkigai.dig@fortnite.com/Minigame_PlungerDunger/TurretSystem
TurretSystem := module:
    @import_as("/Minigame_PlungerDunger/TurretSystem/projectile.projectile_C")
    projectile<scoped {/playkigai.dig@fortnite.com/Minigame_PlungerDunger}> := class<final>(entity):


    projectile_asset<scoped {/playkigai.dig@fortnite.com/Minigame_PlungerDunger}>:entity_prefab = external {}

    @import_as("/Minigame_PlungerDunger/TurretSystem/turret.turret_C")
    turret<scoped {/playkigai.dig@fortnite.com/Minigame_PlungerDunger}> := class<final>(entity):


    turret_asset<scoped {/playkigai.dig@fortnite.com/Minigame_PlungerDunger}>:entity_prefab = external {}

    TurretUiSample<scoped {/playkigai.dig@fortnite.com/Minigame_PlungerDunger}> := class<public>(widget):


    TurretUiSample_component<scoped {/playkigai.dig@fortnite.com/Minigame_PlungerDunger}> := class<final><public>(component):


# Module import path: /playkigai.dig@fortnite.com/Minigame_PlungerDunger/Meshes
Meshes := module:
    Arrow_6F4A9F86<scoped {/playkigai.dig@fortnite.com/Minigame_PlungerDunger}> := class<final><scoped {/playkigai.dig@fortnite.com/Minigame_PlungerDunger}>(mesh_component):
        @editable
        var Material_0<public>:material = external {}

    Arrow_6F4A9F86_asset<scoped {/playkigai.dig@fortnite.com/Minigame_PlungerDunger}>:mesh = external {}

    Box_12680BB2<scoped {/playkigai.dig@fortnite.com/Minigame_PlungerDunger}> := class<final><scoped {/playkigai.dig@fortnite.com/Minigame_PlungerDunger}>(mesh_component):
        @editable
        var Material_0<public>:material = external {}

    Box_12680BB2_asset<scoped {/playkigai.dig@fortnite.com/Minigame_PlungerDunger}>:mesh = external {}

    Box_4A814F4B<scoped {/playkigai.dig@fortnite.com/Minigame_PlungerDunger}> := class<final><scoped {/playkigai.dig@fortnite.com/Minigame_PlungerDunger}>(mesh_component):
        @editable
        var Material_0<public>:material = external {}

    Box_4A814F4B_asset<scoped {/playkigai.dig@fortnite.com/Minigame_PlungerDunger}>:mesh = external {}

    Path_16ABD218<scoped {/playkigai.dig@fortnite.com/Minigame_PlungerDunger}> := class<final><scoped {/playkigai.dig@fortnite.com/Minigame_PlungerDunger}>(mesh_component):
        @editable
        var Material_0<public>:material = external {}

    Path_16ABD218_asset<scoped {/playkigai.dig@fortnite.com/Minigame_PlungerDunger}>:mesh = external {}

    ProjectileMesh<scoped {/playkigai.dig@fortnite.com/Minigame_PlungerDunger}> := class<final><scoped {/playkigai.dig@fortnite.com/Minigame_PlungerDunger}>(mesh_component):
        @editable
        var Material_0<public>:material = external {}

    ProjectileMesh_asset<scoped {/playkigai.dig@fortnite.com/Minigame_PlungerDunger}>:mesh = external {}