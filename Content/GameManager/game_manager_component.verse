<#
@UUID: 5a24cfd3-40f7-42b7-9899-d4ab8bbbf0b8
@Task: GameManagerModule
@Description: A module which handles high level logic for the gameplay systems.
@Tags: Round State, Game State
@Export: true
#>
<#
Ensure IslandSettings match:
    - Teams: Team Index (2)
    - Show Top Center Score HUD: true

Entity Prefab setup for `game_manager` entity prefab:
    - game_manager
        - NOTE: 
            game_manager is an entity prefab, and must be placed in the world with the following components.
        - game_manager_component
            - NOTE:
                game_manager_component requires **all** fields to be non-null as well
                as all other dependant entities to be property setup in order for this component
                to work as intended.
#>
using { /Verse.org }
using { /Verse.org/Native }
using { /Verse.org/SceneGraph }
using { /Verse.org/SceneGraph/KeyframedMovement }
using { /Verse.org/Simulation }
using { /Verse.org/Random }
using { /Fortnite.com/Playspaces }
using { /Fortnite.com/Devices }
using { TargetSystem }
using { TurretSystem }
using { TurretSystem.TurretSystemModule }
using { TargetSystem.TargetSystemModule }

GameManagerModule<public> := module:

    round_state<public> := enum:
        Paused
        Started
        Ended

    game_manager_component<public> := class<final_super>(component):
            # Score manager for player one. 
            # This is activated whenever a target reaches Player Two's goal.
            # Set `Activating Team` to `Team Index`. Ensure its value is set to `1`.
            @editable 
            PlayerOneScoreManager<public>: score_manager_device = score_manager_device{}
            # Score manager for player two.
            # This is activated whenever a target reaches Player One's goal.
            # Set `Activating Team` to `Team Index`. Ensure its value is set to `2`.
            @editable
            PlayerTwoScoreManager<public>: score_manager_device = score_manager_device{}
            @editable
            TimerBillboard : billboard_device = billboard_device{}
            @editable
            TurretOne: turret = turret{}
            @editable
            TurretTwo: turret = turret{}
            @editable
            SpawnDelay: float = 1.5            
            @editable
            RoundTime: int = 60
            
            RoundStarted<public>: event() = event(){}
            RoundPaused<public>: event() = event(){}
            RoundEnded<public>: event() = event(){}

            var Rails: []rail = array{}
            var PilotOne: ?agent = false
            var PilotTwo: ?agent = false
            var RoundTimer : int = 0
            var RoundState: round_state = round_state.Paused


        OnBeginSimulation<override>():void =
            (super:)OnBeginSimulation()
            PlayerOneScoreManager.SetScoreAward(1)
            PlayerTwoScoreManager.SetScoreAward(1)

            if. set TurretOne.GetComponent[turret_component].MaybeGameManager = option. Self 
            if. set TurretTwo.GetComponent[turret_component].MaybeGameManager = option. Self


        OnSimulate<override>()<suspends>:void =
            if(SimRoot := Entity.GetSimulationEntity[]):
                set Rails = for (Rail : SimRoot.FindDescendantEntities(rail)) do Rail
                sync:
                    loop:
                        Sleep(SpawnDelay)
                        AttemptSpawnTarget(SimRoot)

                    loop:
                        RoundStarted.Await()
                        OnRoundStarted()
                        RoundEnded.Await()
                        OnEndRound()


        AttemptSpawnTarget(Root:entity):void= 
            set PilotOne = TurretOne.IsOccupied[] or false
            set PilotTwo = TurretTwo.IsOccupied[] or false

            if(RoundState = round_state.Ended):
                return

            if(not (PilotOne? or PilotTwo?)):
                if(RoundState <> round_state.Ended):
                    set RoundState = round_state.Paused
                    RoundPaused.Signal()
                return

            if(TeamCollection := Entity.GetPlayspaceForEntity[].GetTeamCollection(), Teams := TeamCollection.GetTeams()):
                if:
                    MaybePilotOne := PilotOne?
                    TeamCollection.AddToTeam[MaybePilotOne, Teams[0]]
                if:
                    MaybePilotTwo := PilotTwo?
                    TeamCollection.AddToTeam[MaybePilotTwo, Teams[1]]

            if:
                Rail := Rails[GetRandomInt(0, Rails.Length - 1)]
                Target := target{}
                TargetComp := Target.GetComponent[target_component]
            then:
                set TargetComp.MaybePlayerOne = PilotOne
                set TargetComp.MaybePlayerTwo = PilotTwo
                set TargetComp.MaybeGameManager = option. Self

                Rail.AddEntities(array. Target)
                Target.SetGlobalTransform(Rail.GetGlobalTransform())
            else:
                Err("Unable to spawn target!")


        OnRoundStarted()<suspends>:void=
            set RoundTimer = RoundTime

            race:
                RoundEnded.Await()

                loop:
                    if(RoundState = round_state.Paused):
                        Print("Paused game.")
                        RoundStarted.Await()
                    else if(RoundState = round_state.Ended):
                        break
                    else:
                        Sleep(1.0)

                        set RoundTimer -= 1

                        TimerBillboard.SetText(RoundTimerText(RoundTimer))
                        TimerBillboard.UpdateDisplay()

                        if(RoundTimer <= 0 ):
                            RoundEnded.Signal()
                            set RoundState = round_state.Ended
                            break


        OnEndRound():void=
            if (MaybePilotOne := PilotOne?, MaybePilotTwo := PilotTwo?):
                if(PlayerOneScoreManager.GetCurrentScore(MaybePilotOne) > PlayerTwoScoreManager.GetCurrentScore(MaybePilotTwo)):
                    TimerBillboard.SetText(RoundWinText(MaybePilotOne))
                else if(PlayerOneScoreManager.GetCurrentScore(MaybePilotOne) = PlayerTwoScoreManager.GetCurrentScore(MaybePilotTwo)):
                    TimerBillboard.SetText(RoundDrawText)
                else:
                    TimerBillboard.SetText(RoundWinText(MaybePilotTwo))
            else:
                TimerBillboard.SetText(PlayerMissingText)



        RoundTimerText<localizes>(Time:int):message= "{Time}"
        RoundWinText<localizes>(Agent:agent):message= "{Agent} wins!"
        RoundDrawText<localizes>:message= "Draw!"
        PlayerMissingText<localizes>:message= "A Player is missing!"

        AgentNameText<localizes>(Agent:agent):message= "{Agent}"
        ToString(Agent:agent):string= Localize(AgentNameText(Agent))

        

    (GameManagerComp:game_manager_component).StartRound<public>():void=
        if(GameManagerComp.RoundState <> round_state.Started):
            set GameManagerComp.RoundState = round_state.Started
            GameManagerComp.RoundStarted.Signal()


    # Since PlayerOneScoreManager and PlayerTwoScoreManager are set to use Team Index, 
    # only the agent's associated team can receive an award from the activation.
    # This means it is not necessary to determine which `score_manager_device`
    # to activate prior to awarding score.
    (GameManagerComp:game_manager_component).AddScore<public>(Agent:agent):void=
        if(GameManagerComp.RoundState <> round_state.Started):
            return

        PlayerOneScoreMgr := GameManagerComp.PlayerOneScoreManager
        PlayerOneScoreMgr.Activate(Agent)
        PlayerTwoScoreMgr := GameManagerComp.PlayerTwoScoreManager
        PlayerTwoScoreMgr.Activate(Agent)
        
        
    (GameManagerComp:game_manager_component).GetRoundState<public>()<transacts>:round_state=
        GameManagerComp.RoundState