<#
@UUID: fde80742-842f-4b1d-96fc-09edca715f9e
@Task: TargetSystemModule
@Description: A Module that handles the position, spawning, moving, and collision events of the targets and rails of the minigame.
@Tags: Instantiate Prefab, Collisions, Overlaps, Keyframes, TickEvents
@Export: true
#>
<#
Entity Prefab setup for the `rail` entity prefab:
    `rail` needs to have a top level entity with a mesh component the mesh should be long enough to be your rail.
        The pivot point should be set to center. (Usually done by default)
        Transform component
    `rail` should have two children entities, one on either end of the rail mesh both with transform components.
        pick one side of the rail and add the target_component_player_one_goal (This will be used for keyframes later)
        pick the other side of the rail and add the target_component_player_two_goal (This will be used for keyframes later)

    - rail
        - mesh_component
        - transform_component
        - entity
            - target_component_player_one_goal
            - transform_component (One end of the rail)
        -entity
            - target_component_player_two_goal
            - transform_component (Other end of the rail)
            
The "rail" entity prefab **must** be placed in the scene whereever the creator wants them.
        
Entity Prefab setup for the "target" entity prefab:
    "target" has one entity with a mesh and transform component.
        The pivot point for the mesh should be at the bottom. (This can be updated using the Modeling Mode(Shift + 4) -> Xform -> Edit Pivot -> Bottom)
        The mesh also needs to have a collision added with OverlapAll set.
    
    - target
        - mesh_component
        - transform_component
        
Entity Prefab setup for the "explosion" entity prefab:
    "explosion" has a transform_component and a partical_system_component with a confetti burst niagra system.
    
    - explosion
        - particle_system_component
        - transform_component
#>
using { /Verse.org/Random }
using { /Verse.org/SceneGraph }
using { /Verse.org/SceneGraph/KeyframedMovement }
using { /Verse.org/Simulation }
using { /Verse.org/SpatialMath }
using { /Fortnite.com/Devices }
using { /Fortnite.com/Playspaces }
using { TurretSystem.ProjectileModule }
using { TurretSystem }
using { TurretSystem.TurretSystemModule }
using { GameManager.GameManagerModule }

TargetSystemModule<public> := module:

    target_component_player_one_goal<public> := class<final_super>(component):
    target_component_player_two_goal<public> := class<final_super>(component):

    target_component<public> := class<final_super>(component):

        RemovedFromScene : event() = event(){}
        HitByPlayer: event(int) = event(int){}

        var MaybeGameManager<public>: ?game_manager_component = false
        var MaybePlayerOne<public>: ?agent = false
        var MaybePlayerTwo<public>: ?agent = false

        var GoingToPlayerOne: logic = true
        var PlayerOneGoals: []entity = array{}
        var PlayerTwoGoals: []entity = array{}
        
        var MaybePostPhysics: ?cancelable = false


        OnBeginSimulation<override>():void =
            (super:)OnBeginSimulation()

            set PlayerOneGoals = for (End : Entity.GetParent[].FindDescendantEntitiesWithComponent(target_component_player_one_goal)) do End
            set PlayerTwoGoals = for (End : Entity.GetParent[].FindDescendantEntitiesWithComponent(target_component_player_two_goal)) do End
            set GoingToPlayerOne = ShouldGoToPlayerOne()


        ShouldGoToPlayerOne():logic=
            if(MaybePlayerOne? and MaybePlayerTwo?):
                if (GetRandomInt(0,1) = 0) then true else false
            else if(MaybePlayerTwo?):
                false
            else true


        CreateKeyFrame(Goals:[]entity):keyframed_movement_delta=
            if(Goal := Goals[0]):
                DeltaTransform := transform:
                    Translation := Goal.GetLocalTransform().Translation - Entity.GetLocalTransform().Translation
                    Scale := vector3{}
                   
                Keyframe := keyframed_movement_delta:
                    Transform := DeltaTransform
                    Duration := GetRandomFloat(3.0, 10.0)
            else keyframed_movement_delta{}


        OnSimulate<override>()<suspends>:void=
            if(MovementComponent := Entity.GetComponent[keyframed_movement_component], GameManagerComp := MaybeGameManager?):
                var Keyframe: keyframed_movement_delta = keyframed_movement_delta{}
                set Keyframe = if(GoingToPlayerOne?). CreateKeyFrame(PlayerOneGoals) else CreateKeyFrame(PlayerTwoGoals)

                MovementComponent.SetKeyframes(array. Keyframe, oneshot_keyframed_movement_playback_mode{})
                MovementComponent.Play()

                race:
                    GameManagerComp.RoundEnded.Await()
                    RemovedFromScene.Await()
                    MovementComponent.FinishedEvent.Await()

                    loop:
                        MoveToPlayerIndex := if. GoingToPlayerOne? then 0 else 1

                        PlayerHitIndex := HitByPlayer.Await()

                        if(PlayerHitIndex = MoveToPlayerIndex):
                            set GoingToPlayerOne = logic. not GoingToPlayerOne?

                            set Keyframe = if(GoingToPlayerOne?). CreateKeyFrame(PlayerOneGoals) else CreateKeyFrame(PlayerTwoGoals)

                            MovementComponent.SetKeyframes(array. Keyframe, oneshot_keyframed_movement_playback_mode{})
                            MovementComponent.Play()
                            
                    loop:
                        GameManagerComp.RoundPaused.Await()
                        MovementComponent.Pause()
                        GameManagerComp.RoundStarted.Await()
                        MovementComponent.Play()


                Entity.RemoveFromParent()


        OnRemovingFromScene<override>():void=
            if(GameManagerComp := MaybeGameManager?):
                if:
                    GoingToPlayerOne?
                    PlayerTwo := MaybePlayerTwo?
                then:
                    GameManagerComp.AddScore(PlayerTwo)
                    Print("Awarded point to Player Two.")
                else if:
                    not GoingToPlayerOne?
                    PlayerOne := MaybePlayerOne?
                then:
                    GameManagerComp.AddScore(PlayerOne)
                    Print("Awarded point to Player One.")

            if(SimRoot := Entity.GetSimulationEntity[]):
                Explosion := explosion{}
                SimRoot.AddEntities(array. Explosion)
                Explosion.SetGlobalTransform(Entity.GetGlobalTransform())

            RemovedFromScene.Signal()

                
        OnEndSimulation<override>():void=
            if. PostPhysics := MaybePostPhysics? then PostPhysics.Cancel()



    (TargetComp:target_component).TriggerHit<public>(Agent:?agent):void=
        if(TargetComp.MaybePlayerOne = Agent):
            TargetComp.HitByPlayer.Signal(0)
        else if(TargetComp.MaybePlayerTwo = Agent):
            TargetComp.HitByPlayer.Signal(1)
        