<#
@UUID: f82ddd21-029c-486e-80e3-86496a10797a
@Task: ProjectileModule
@Description: A module that handles that handles physics calculations for a projectile.
@Tags: Instantiate Prefab, Physics, Projectile, Bullet
@Export: true
#>
<#
Entity Prefab setup for the `projectile` entity prefab:
    `projectile` is an entity prefab that must have a `mesh_component` attached
    as well as a `projectile_component`.

    `projectile` is to be instantiated whenever a player attempts to fire their turret. `projectile` follows
    a parabolic arc dependant on the `Gravity`, `Velocity`, and `Drag` forces applied to it. 

    When spawning a `projectile`, the spawning entity must spawn the `projectile` at an appropriate location
    (the barrel/spawn point of the turret, or the ViewLocation of the player that is spawning it).

    - projectile
        - transform_component
        - mesh_component
            - NOTE:
                Ensure `Collision Presets` is set to `OverlapAll`.
        - projectile_component
            NOTE:
                The following **must** be set when spawning a projectile prefab:
                    - Velocity
                    - Lifetime
                    - OwnerAgent

Entity Prefab setup for the "explosion" entity prefab:
    "explosion" has a transform_component and a partical_system_component with a confetti burst niagra system.
    
    - explosion
        - particle_system_component
        - transform_component
#>
using { /Verse.org }
using { /Verse.org/Native }
using { /Verse.org/SceneGraph }
using { /Verse.org/Simulation }
using { /Verse.org/SpatialMath }
using { TargetSystem }
using { TargetSystem.TargetSystemModule }

ProjectileModule<public> := module:

    projectile_component<public> := class<final_super>(component):

        # Constant force applied to projectile while in flight.
        @editable
        Gravity: vector3 = vector3:
            Left := 0.0
            Up := -8.0
            Forward := 0.0
        # Force which reduces projectile speed while in flight.
        @editable
        Drag: float = 0.05
        # Lifetime of projectile. Once this time has expired, the projectile 
        # will be destroyed.
        # NOTE: Must be set when spawning this entity.
        var Lifetime<public>: float = 0.0

        # Current velocity of this projectile (world space).
        # NOTE: Must be set when spawning this entity.
        var Velocity<public>: vector3 = vector3{}
        var TransformComp: ?transform_component = false
        
        # Agent that fired this projectile.
        # NOTE: Must be set when spawning this entity.
        var OwnerAgent<public>: ?agent = false

        var FrameDelta: float = 0.0
        var PrevSimTime: float = 0.0
    

        OnBeginSimulation<override>():void=
            (super:)OnBeginSimulation()

            if(MaybeTransform := Entity.GetComponent[transform_component]):
                set TransformComp = option. MaybeTransform
            else:
                Print("Missing `transform_component`!")
            

        OnSimulate<override>()<suspends>:void=
            loop:
                set FrameDelta = GetSimulationElapsedTime() - PrevSimTime
                
                if. Lifetime <= 0.0 then. Entity.RemoveFromParent() else. set Lifetime -= FrameDelta

                if(MaybeTransformComp := TransformComp?):
                    CurrentTransform := Entity.GetGlobalTransform()
                    set Velocity = ((Velocity + Gravity * FrameDelta) * (1.0 - (Drag * FrameDelta)))

                    NewTransform := transform:
                        Rotation := IdentityRotation()
                        Translation := CurrentTransform.Translation + Velocity
                        Scale := CurrentTransform.Scale

                    Entity.SetGlobalTransform(NewTransform)

                HandleCollisions()

                set PrevSimTime = GetSimulationElapsedTime()

                Sleep(0.0)
                

        # Must return early to act only on first filtered collision event.
        HandleCollisions():void=    
            for:
                Hit: Entity.FindOverlapHits()
                Target := Hit.TargetComponent.Entity.GetComponent[target_component]
            do:
                Target.TriggerHit(OwnerAgent)

                if(SimRoot := Entity.GetSimulationEntity[]):
                    Explosion := explosion{}
                    SimRoot.AddEntities(array. Explosion)
                    Explosion.SetGlobalTransform(Entity.GetGlobalTransform())

                Entity.RemoveFromParent()
                return
                