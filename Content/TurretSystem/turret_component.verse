<#
@UUID: 0efd30ae-7bc5-45db-a9ba-fd6116ce3e2b
@Task: TurretSystemModule
@Description: A module which handles the logic for a stationary turret entity to allow a player to enter, pilot, and fire projectiles.
@Tags: Turret, Vehicle, Stationary Turret, Player Pilot, Projectile, Instantiate Prefab
@Export: true
#>
<#
Entity Prefab setup for the `turret` entity prefab:
    The `turret` entity prefab must be placed in the world. All of its fields must have valid references in order
    to work as intended.

    - turret
        - mesh_component
            - NOTE: 
                Respresents the physical body of the turret. Required to allow the turret to be interacted with
                using the `interactable_component` parent class.
        - turret_component
#>
using { /Verse.org }
using { /Verse.org/Native }
using { /Verse.org/SceneGraph }
using { /Verse.org/Simulation }
using { /Fortnite.com/Devices }
using { /Fortnite.com/Characters }
using { /UnrealEngine.com/Temporary/SpatialMath }
using { /Fortnite.com/FortPlayerUtilities }
using { /UnrealEngine.com }
using { ProjectileModule }
using { GameManager.GameManagerModule }

TurretSystemModule<public> := module:

    turret_component<public> := class<final_super>(interactable_component):
        <# 
        Camera used to represent the turret's point of view.
        Ensure the following settings match:
            - Add to Players on Start: false.
            - Distance: 0
            - Offset X: 0
            - Offset Y: 0
            - Offset Z: 0
            - Boom Collision: false
        #>
        @editable
        TurretCamera: gameplay_camera_orbit_device = gameplay_camera_orbit_device{}
        @editable
        FireTurretInput: input_trigger_device = input_trigger_device{}
        @editable
        ExitTurretInput: input_trigger_device = input_trigger_device{}

        @editable
        MinCharge: float = 15.0
        @editable
        MaxCharge: float = 40.0
        @editable
        ChargeTime: float = 0.75
        @editable
        TurretCooldown: float = 0.75
        # Seconds before a spawned projectile will destroy itself.
        # Passed to a `projectile_component` when spawning a new `projectile`.
        @editable
        ProjectileLifetime: float = 5.0


        var TurretCooldownTimer: float = 0.0
        # Value representing the force to be applied to the current projectile.
        var ProjectileForce: float = 0.0
        # Value used to determine the current projectile force when lerping between 
        # MinCharge and MaxCharge.
        var ChargeDelta: float = 0.0

        # Local offset of the agent when piloting the turret.
        # Default of `Y` is 200.0 to ensure enough vertical clearance.
        # This must be adjusted to allow the player to be moved to the correct location
        # determined by the turret's mesh.
        @editable
        PilotOffset: vector3 = vector3{Y:=200.0}
        var Pilot<public>: ?agent = false
        var PilotChar: ?fort_character = false

        var FrameDelta: float = 0.0
        var PrevSimTime: float = 0.0        
        
        var MaybeGameManager<public>: ?game_manager_component = false

    
        OnBeginSimulation<override>():void=
            (super:)OnBeginSimulation()

            SucceededEvent.Subscribe(OnAgentEnterTurret)
            ExitTurretInput.PressedEvent.Subscribe(OnAgentExitTurret)


        OnSimulate<override>()<suspends>:void=
            loop:
                set FrameDelta = GetSimulationElapsedTime() - PrevSimTime
                set TurretCooldownTimer -= FrameDelta

                if:
                    MaybePilot := Pilot?
                    MaybeChar := PilotChar?
                    FireTurretInput.IsHeld[MaybePilot]
                then:
                    if(ChargeDelta < 1.0 and TurretCooldownTimer <= 0.0):
                        set ChargeDelta += (FrameDelta / ChargeTime)
                        set ChargeDelta = Clamp(ChargeDelta, 0.0, 1.0)
                        set ProjectileForce = Lerp(MinCharge, MaxCharge, ChargeDelta)
                else:
                    if(ProjectileForce > 0.0):
                        SpawnProjectile()

                        set TurretCooldownTimer = TurretCooldown
                        set ChargeDelta = 0.0
                        set ProjectileForce = 0.0 

                set PrevSimTime = GetSimulationElapsedTime()

                Sleep(0.0)


        OnAgentEnterTurret(Agent:agent):void=
            if(Pilot?):
                return
            
            if:
                FortChar := Agent.GetFortCharacter[]
                TurretLocation := Entity.GetGlobalTransform().Translation
                PilotLocation := vector3:
                    X := TurretLocation.Forward + PilotOffset.Z
                    Y := -TurretLocation.Left + PilotOffset.X
                    Z := TurretLocation.Up + PilotOffset.Y
                FortChar.TeleportTo[PilotLocation, IdentityRotation()]
            then:                                
                TurretCamera.AddTo(Agent)

                FortChar.PutInStasis(stasis_args{AllowTurning:=true, AllowFalling:=false, AllowEmotes:=false})
                FortChar.Hide()

                set Pilot = option. Agent
                set PilotChar = option. FortChar
                Print("{Agent} entered turret.")

                if(GameManagerComp := MaybeGameManager?):
                    GameManagerComp.StartRound()

        
        OnAgentExitTurret(Agent:agent):void=
            if(not Pilot = option. Agent):
                return

            TurretCamera.RemoveFrom(Agent)
            
            if(MaybeChar := PilotChar?):
                MaybeChar.ReleaseFromStasis()
                MaybeChar.Show()

                set Pilot = false
                set PilotChar = false         
                Print("{Agent} exited turret.")


        SpawnProjectile():void=
            if:
                MaybeChar := PilotChar?
                SimRoot := Entity.GetSimulationEntity[]
            then:
                ViewRotation := MaybeChar.GetViewRotation()
                SpawnLocation := MaybeChar.GetViewLocation() + ViewRotation.GetLocalForward()
                Direction := ViewRotation.GetLocalForward()
                LaunchVelocity := Direction * ProjectileForce
                Projectile := projectile{}

                SimRoot.AddEntities(array. Projectile)

                Projectile.SetGlobalTransform(FromTransform(transform{Translation:=SpawnLocation}))
                if(ProjectileComp := Projectile.GetComponent[projectile_component]):
                    set ProjectileComp.Velocity = FromVector3(LaunchVelocity)
                    set ProjectileComp.Lifetime = ProjectileLifetime
                    set ProjectileComp.OwnerAgent = option. Pilot?

                if (RoundedPercent := Round[100.0 * ((ProjectileForce - MinCharge) / (MaxCharge - MinCharge))]):
                    Print("Fired at {RoundedPercent}% charge.")



        InteractMessage<override>(Agent:agent)<reads><decides>:message= LocText<localizes>:message="Enter Turret"

        AgentNameText<localizes>(Agent:agent):message= "{Agent}"
        ToString(Agent:agent):string= Localize(AgentNameText(Agent))


                
    (Turret:turret).IsOccupied<public>()<decides><transacts>:?agent=
        TurretComp := Turret.GetComponent[turret_component]
        TurretComp.Pilot
        