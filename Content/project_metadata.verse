<#
@UUID: b7447af2-9060-4a52-9fbe-f0c10750c70d
@Task: Plunger Dunger
@Description: Plunger Dunger is a two-player competitive minigame where players fire physical projectiles from stationary turrets in an attempt to hit moving targets. Targets travel along guided rails that span between each turret. If a target is hit by a player, it begins traveling to the opposing player's goal. Players are awarded points whenever a target fully reaches the opposing player's goal. Players must time their shots as the turret has a cooldown between firing.
@Tags: Turret, Target, Projectile, Score, Multiplayer, Two-Player, Competitive
@Export: true
#>